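; Event component that feeds progress deltas into the owner's mesh-networked world reactions component, which it looks up and caches on BeginPlay.
; UpdateValue acts only when the mesh network node type is Root and the cached component is valid.
; It scales the incoming delta by the current multiplier, rounds up, adds it to the current value and caps it at the required value.
; It then calls ValueUpdated with that total, whether the requirement is met, and the required value.
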(event EventBeginPlay
  (bind _returnvalue (Components|GetOwner))
  (bind _returnvalue_2 (Actor|GetComponentbyClass _returnvalue "/Game/Quests/MeshNetwork/BP_MeshNetworkedWorldReactionsComponent.BP_MeshNetworkedWorldReactionsComponent_C"))
  ;[0 0 0 0.5] BeginPlay
  (Utilities|IsValid _returnvalue_2
    (:Is%20Valid
      (Variables|Default|SetCachedMeshNetworkedWorldReactionsComponent _returnvalue_2))))

(event Custom|UpdateValue
  (params (NewValue :type Integer64))
  (bind _cachedmeshnetworkedworldreactionscomponent (Variables|Default|GetCachedMeshNetworkedWorldReactionsComponent))
  ;[0 0 0 0.5] OnWorldReactionTriggered
  (switch Utilities|FlowControl|Switch|SwitchonEMeshNetworkNodeType (MeshNetworking|GetMeshNetworkNodeType)
    (:Root
      (Utilities|IsValid _cachedmeshnetworkedworldreactionscomponent
        (:Is%20Valid
          (Class|BPMeshNetworkedWorldReactionsComponent|ValueUpdated _cachedmeshnetworkedworldreactionscomponent (Math|Integer64|Min(Integer64) (+ (Math|Float|CeiltoInteger64 (* (Math|Conversions|ToFloat(Integer64) NewValue) (Variables|/Script/FortniteGame.FortGameFrameworkComponent_MeshNetworkedEvents|GetCurrentDeltaMultiplier (Variables|Default|GetCachedMeshNetworkedWorldReactionsComponent)))) (Variables|/Script/FortniteGame.FortGameFrameworkComponent_MeshNetworkedEvents|GetCurrentValue (Variables|Default|GetCachedMeshNetworkedWorldReactionsComponent))) (Variables|/Script/FortniteGame.FortGameFrameworkComponent_MeshNetworkedEvents|GetCurrentRequiredValue (Variables|Default|GetCachedMeshNetworkedWorldReactionsComponent))) (>= (Math|Integer64|Min(Integer64) (+ (Math|Float|CeiltoInteger64 (* (Math|Conversions|ToFloat(Integer64) NewValue) (Variables|/Script/FortniteGame.FortGameFrameworkComponent_MeshNetworkedEvents|GetCurrentDeltaMultiplier (Variables|Default|GetCachedMeshNetworkedWorldReactionsComponent)))) (Variables|/Script/FortniteGame.FortGameFrameworkComponent_MeshNetworkedEvents|GetCurrentValue (Variables|Default|GetCachedMeshNetworkedWorldReactionsComponent))) (Variables|/Script/FortniteGame.FortGameFrameworkComponent_MeshNetworkedEvents|GetCurrentRequiredValue (Variables|Default|GetCachedMeshNetworkedWorldReactionsComponent))) (Variables|/Script/FortniteGame.FortGameFrameworkComponent_MeshNetworkedEvents|GetCurrentRequiredValue (Variables|Default|GetCachedMeshNetworkedWorldReactionsComponent)))))))))
