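; Mesh network component for quest world-reaction events, keeping a shared event's progress in mesh metadata keyed by its event ID tag.
; Triggered reactions pass their delta to the owner's cached event mesh network component, which RPCs to the root server.
; On the root, value updates write progress and completion to metadata.
; Metadata changes refresh current value, requirement and multiplier, with default initialization as fallback.
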(event EventBeginPlay
  (bind _returnvalue (Components|GetOwner))
  (bind _returnvalue_2 (Actor|GetComponentbyClass _returnvalue "/Game/Quests/MeshNetwork/BP_EventMeshNetworkComponent.BP_EventMeshNetworkComponent_C"))
  ;[0 0 0 0.5] BeginPlay
  (Utilities|IsValid _returnvalue_2
    (:Is%20Valid
      (Variables|Default|SetCachedEventMeshNetworkComponent _returnvalue_2)
      (bind _nodetype (Class|FortAsyncActionMeshNetworkReady|MeshNetworkReadyAsync)
        (:OnReady
          (switch Utilities|FlowControl|Switch|SwitchonEMeshNetworkNodeType _nodetype
            (:Root
              (MeshNetworkSubsystem|BindEventtoOnMeshMetaDataUpdated (GameInstanceSubsystems|GetMeshNetworkSubsystem) (EventDispatchers|CreateEvent :!event Custom|UpdateMeshMetadata))
              (CallFunction|UpdateMeshMetadata))))))))

(event EventOnWorldReactionTriggered
  (params (Delta :type Integer64))
  (bind _cachedeventmeshnetworkcomponent (Variables|Default|GetCachedEventMeshNetworkComponent))
  ;[0 0 0 0.5] OnWorldReactionTriggered
  (Utilities|IsValid _cachedeventmeshnetworkcomponent
    (:Is%20Valid
      ;;[0 0 0 0.5] RPC to the Root Server
      (Class|BPEventMeshNetworkComponent|UpdateValue _cachedeventmeshnetworkcomponent Delta))))

(event Custom|ValueUpdated
  (params (NewValue :type Integer64) (EventCompleted :type Boolean))
  ;[0 0 0 0.5] ValueUpdated
  (if (CallFunction|IsRootServer)
    (bind _output_get (Variables|FortGameFrameworkComponent_MeshNetworkedEvents|SetCurrentValue NewValue))
    (bind _returnvalue (MeshNetworkSubsystem|GetMetaDatawithKey (GameInstanceSubsystems|GetMeshNetworkSubsystem) (GameplayTags|GetTagName (CallFunction|GetEventID)) (Variables|Default|GetCurrentMeshNetworkMetadata)))
    (if _returnvalue
      (bind _structout (Utilities|Struct|SetmembersinMeshNetworkedEventMetadata (Variables|Default|GetCurrentMeshNetworkMetadata) _output_get EventCompleted))
      (MeshNetworkSubsystem|SetMetaDatawithKey (GameInstanceSubsystems|GetMeshNetworkSubsystem) (GameplayTags|GetTagName (CallFunction|GetEventID)) _structout)
      (else
        ;; Shouldn't ever happen
        (CallFunction|UpdateMeshMetadata)))))

(event Custom|UpdateMeshMetadata
  (bind _currentmeshnetworkmetadata (Variables|Default|GetCurrentMeshNetworkMetadata))
  (bind (_currentvalue _requiredvalue _deltamultiplier _bhasinitialized) (Utilities|Struct|BreakMeshNetworkedEventMetadata _currentmeshnetworkmetadata))
  (bind _currentmeshnetworkmetadata_2 (Variables|Default|GetCurrentMeshNetworkMetadata))
  (bind (_currentvalue_2 _requiredvalue_2 _deltamultiplier_2) (Utilities|Struct|BreakMeshNetworkedEventMetadata _currentmeshnetworkmetadata_2))
  ;[0 0 0 0.5] MeshMetadataChanged
  (bind _nodetype (Class|FortAsyncActionMeshNetworkReady|MeshNetworkReadyAsync)
    (:OnReady
      (switch Utilities|FlowControl|Switch|SwitchonEMeshNetworkNodeType _nodetype
        (:Root
          (bind _returnvalue (MeshNetworkSubsystem|GetMetaDatawithKey (GameInstanceSubsystems|GetMeshNetworkSubsystem) (GameplayTags|GetTagName (CallFunction|GetEventID)) _currentmeshnetworkmetadata))
          (if _returnvalue
            (if _bhasinitialized
              (Variables|FortGameFrameworkComponent_MeshNetworkedEvents|SetCurrentValue _currentvalue_2)
              (Variables|FortGameFrameworkComponent_MeshNetworkedEvents|SetCurrentRequiredValue _requiredvalue_2)
              (Variables|FortGameFrameworkComponent_MeshNetworkedEvents|SetCurrentDeltaMultiplier _deltamultiplier_2)
              (else
                _)))
          (CallFunction|InitializeMeshNetworkData))))))

(event Custom|InitializeMeshNetworkData
  ;[0 0 0 0.5] InitializeMeshNetworkData
  (bind _nodetype (Class|FortAsyncActionMeshNetworkReady|MeshNetworkReadyAsync)
    (:OnReady
      (switch Utilities|FlowControl|Switch|SwitchonEMeshNetworkNodeType _nodetype
        (:Root
          (MeshNetworkSubsystem|SetMetaDatawithKey (GameInstanceSubsystems|GetMeshNetworkSubsystem) (GameplayTags|GetTagName (CallFunction|GetEventID)) (Utilities|Struct|MakeMeshNetworkedEventMetadata (Variables|FortGameFrameworkComponent_MeshNetworkedEvents|GetCurrentValue) (Variables|Default|GetDefaultRequiredValue) (Variables|Default|GetDefaultDeltaMultiplier) true)))))))
